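# Sound cue for a Husk monster body dematerializing on death.
# A Random node picks one of six equally weighted dematerialize waves, then a Modulator varies its pitch widely around normal.
# Playback uses the shared Husk attenuation asset and an enemy-damaged sound class, at reduced volume.
# A concurrency override limits how many instances play at once.
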SoundCue /Game/Sounds/Fort_Monster_Husk/SFX/Husk_Body_DeathDematerial_Cue
root=Modulator_0
nodes [
 {id=Modulator_0, children [Random_0], properties {pitchMax=1.3, pitchMin=0.7}}
 {id=Random_0, children [WavePlayer_0,WavePlayer_1,WavePlayer_2,WavePlayer_3,WavePlayer_4,WavePlayer_5], properties {weights [1,1,1,1,1,1]}}
 {id=WavePlayer_0, properties {soundWaveAssetPtr=&/Husk_Dematerialize_03}}
 {id=WavePlayer_1, properties {soundWaveAssetPtr=&/Husk_Dematerialize_01}}
 {id=WavePlayer_2, properties {soundWaveAssetPtr=&/Husk_Dematerialize_02}}
 {id=WavePlayer_3, properties {soundWaveAssetPtr=&/Husk_Dematerialize_04}}
 {id=WavePlayer_4, properties {soundWaveAssetPtr=&/Husk_Dematerialize_05}}
 {id=WavePlayer_5, properties {soundWaveAssetPtr=&/Husk_Dematerialize_06}}
]
properties {
 attenuationSettings=/Game/Sounds/Attenuation/Fort_Monster_Husk_Att
 bOverrideConcurrency=true
 maxDistance=2800
 soundClassObject=/Game/Sounds/FortSoundClassesAndModes/Fort_Enemy_Damaged_PSM
 volumeMultiplier=0.6
 attenuationOverrides {falloffDistance=5000, attenuationShapeExtents {x=1000}}
 concurrencyOverrides {maxCount=3}
}
