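# Sound cue that plays a randomized body impact sound for Husk monsters, routed to the enemy-damaged sound class.
# A Quality Level root feeds two Modulator-into-Random branches, leaving its other two slots empty.
# The first branch picks evenly from thirteen impact waves, while the second uses a six-wave subset for variation at lower asset count.
# Husk focus attenuation spatializes it, and a concurrency rule allows one instance per owner, stopping the oldest.
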SoundCue /Game/Sounds/Fort_Monster_Husk/SFX/Husk_Body_Impact_Cue
root=QualityLevel_1
nodes [
 {id=Modulator_8, children [Random_9]}
 {id=Modulator_9, children [Random_10]}
 {id=QualityLevel_1, children [Modulator_8,Modulator_9,null,null]}
 {
  id=Random_10
  children [WavePlayer_67, WavePlayer_68, WavePlayer_69, WavePlayer_74, WavePlayer_75, WavePlayer_78]
  properties {weights [1,1,1,1,1,1]}
 }
 {
  id=Random_9
  children [
   WavePlayer_54
   WavePlayer_55
   WavePlayer_56
   WavePlayer_57
   WavePlayer_58
   WavePlayer_59
   WavePlayer_60
   WavePlayer_61
   WavePlayer_62
   WavePlayer_63
   WavePlayer_64
   WavePlayer_65
   WavePlayer_66
  ]
  properties {weights [1,1,1,1,1,1,1,1,1,1,1,1,1]}
 }
 {id=WavePlayer_54, properties {soundWaveAssetPtr=&/husk_body_impact_mix_10}}
 {id=WavePlayer_55, properties {soundWaveAssetPtr=&/husk_body_impact_mix_13}}
 {id=WavePlayer_56, properties {soundWaveAssetPtr=&/husk_body_impact_mix_14}}
 {id=WavePlayer_57, properties {soundWaveAssetPtr=&/husk_body_impact_mix_15}}
 {id=WavePlayer_58, properties {soundWaveAssetPtr=&/husk_body_impact_mix_01}}
 {id=WavePlayer_59, properties {soundWaveAssetPtr=&/husk_body_impact_mix_02}}
 {id=WavePlayer_60, properties {soundWaveAssetPtr=&/husk_body_impact_mix_03}}
 {id=WavePlayer_61, properties {soundWaveAssetPtr=&/husk_body_impact_mix_04}}
 {id=WavePlayer_62, properties {soundWaveAssetPtr=&/husk_body_impact_mix_05}}
 {id=WavePlayer_63, properties {soundWaveAssetPtr=&/husk_body_impact_mix_06}}
 {id=WavePlayer_64, properties {soundWaveAssetPtr=&/husk_body_impact_mix_07}}
 {id=WavePlayer_65, properties {soundWaveAssetPtr=&/husk_body_impact_mix_08}}
 {id=WavePlayer_66, properties {soundWaveAssetPtr=&/husk_body_impact_mix_09}}
 {id=WavePlayer_67, properties {soundWaveAssetPtr=&husk_body_impact_mix_10}}
 {id=WavePlayer_68, properties {soundWaveAssetPtr=&husk_body_impact_mix_13}}
 {id=WavePlayer_69, properties {soundWaveAssetPtr=&husk_body_impact_mix_14}}
 {id=WavePlayer_74, properties {soundWaveAssetPtr=&husk_body_impact_mix_04}}
 {id=WavePlayer_75, properties {soundWaveAssetPtr=&husk_body_impact_mix_05}}
 {id=WavePlayer_78, properties {soundWaveAssetPtr=&husk_body_impact_mix_08}}
]
properties {
 attenuationSettings=/Game/Sounds/Attenuation/Fort_Monster_Husk_Focus_Att
 maxDistance=3500
 soundClassObject=/Game/Sounds/FortSoundClassesAndModes/Fort_Enemy_Damaged_PSM
 volumeMultiplier=1
 attenuationOverrides {falloffDistance=5000, attenuationShapeExtents {x=1000}}
 concurrencySet [/Game/Sounds/Concurrency_Rules/OneMaximum_Owner_StopOldest]
}
