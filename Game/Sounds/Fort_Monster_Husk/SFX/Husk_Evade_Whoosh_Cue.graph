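# Sound cue for the husk monster's evade whoosh, classed as an enemy attack sound.
# A Random node picks one of three equally weighted quick, sharp whoosh waves borrowed from the weapon whoosh set and feeds it through a Modulator.
# The cue boosts volume and overrides both attenuation and concurrency, with a rule that blocks new instances once the limit is reached.
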SoundCue /Game/Sounds/Fort_Monster_Husk/SFX/Husk_Evade_Whoosh_Cue
root=Modulator_3
nodes [
 {id=Modulator_3, children [Random_0]}
 {id=Random_0, children [WavePlayer_0,WavePlayer_1,WavePlayer_2], properties {weights [1,1,1]}}
 {id=WavePlayer_0, properties {soundWaveAssetPtr=/Game/Sounds/Fort_Weapons/Whooshes/Whoosh_QuickSharp_01}}
 {id=WavePlayer_1, properties {soundWaveAssetPtr=/Game/Sounds/Fort_Weapons/Whooshes/Whoosh_QuickSharp_02}}
 {id=WavePlayer_2, properties {soundWaveAssetPtr=/Game/Sounds/Fort_Weapons/Whooshes/Whoosh_QuickSharp_03}}
]
properties {
 attenuationSettings=/Game/Sounds/Attenuation/Fort_Bullet_ImpactOpponent_Lng_Att
 bOverrideAttenuation=true
 bOverrideConcurrency=true
 maxDistance=6000
 soundClassObject=/Game/Sounds/FortSoundClassesAndModes/Fort_Enemy_Attack
 volumeMultiplier=1.5
 attenuationOverrides {falloffDistance=5000, attenuationShapeExtents {x=1000}}
 concurrencyOverrides {resolutionRule=PreventNew}
}
