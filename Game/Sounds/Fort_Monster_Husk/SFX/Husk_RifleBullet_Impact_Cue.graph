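# Sound cue for rifle bullets hitting a Husk, classed as enemy damage, with long opponent-impact attenuation and limited concurrency.
# A quality switch feeds two similar branches, the second drawing on fewer wave variations, and leaves its last two slots empty.
# Each branch waits briefly, then mixes a random incoming bullet whip with a further-delayed random Husk body impact, both pitch-modulated for variety.
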SoundCue /Game/Sounds/Fort_Monster_Husk/SFX/Husk_RifleBullet_Impact_Cue
root=QualityLevel_0
nodes [
 {id=Delay_2, children [Modulator_2], properties {delayMax=0.15, delayMin=0.1}}
 {id=Delay_3, children [Mixer_1], properties {delayMax=0.06, delayMin=0.04}}
 {id=Delay_4, children [Modulator_4], properties {delayMax=0.15, delayMin=0.1}}
 {id=Delay_5, children [Mixer_2], properties {delayMax=0.06, delayMin=0.04}}
 {id=Mixer_1, children [Delay_2,Modulator_3], properties {inputVolume [1,1]}}
 {id=Mixer_2, children [Delay_4,Modulator_5], properties {inputVolume [1,1]}}
 {id=Modulator_2, children [Random_3], properties {pitchMax=1.2, pitchMin=1}}
 {id=Modulator_3, children [Random_2], properties {pitchMax=1.3, pitchMin=0.9}}
 {id=Modulator_4, children [Random_5], properties {pitchMax=1.2, pitchMin=1}}
 {id=Modulator_5, children [Random_4], properties {pitchMax=1.3, pitchMin=0.9}}
 {id=QualityLevel_0, children [Delay_3,Delay_5,null,null]}
 {id=Random_2, children [WavePlayer_11, WavePlayer_12, WavePlayer_13, WavePlayer_14], properties {weights [1,1,1,1]}}
 {
  id=Random_3
  children [WavePlayer_15, WavePlayer_16, WavePlayer_17, WavePlayer_18, WavePlayer_19, WavePlayer_20, WavePlayer_21]
  properties {weights [1,1,1,1,1,1,1]}
 }
 {id=Random_4, children [WavePlayer_22, WavePlayer_23, WavePlayer_24], properties {weights [1,1,1]}}
 {id=Random_5, children [WavePlayer_26, WavePlayer_27, WavePlayer_30, WavePlayer_31], properties {weights [1,1,1,1]}}
 {id=WavePlayer_11, properties {soundWaveAssetPtr=/Game/Sounds/Fort_Weapons/Impacts/bullet_whip_in_01}}
 {id=WavePlayer_12, properties {soundWaveAssetPtr=/Game/Sounds/Fort_Weapons/Impacts/bullet_whip_in_02}}
 {id=WavePlayer_13, properties {soundWaveAssetPtr=/Game/Sounds/Fort_Weapons/Impacts/bullet_whip_in_03}}
 {id=WavePlayer_14, properties {soundWaveAssetPtr=/Game/Sounds/Fort_Weapons/Impacts/bullet_whip_in_04}}
 {id=WavePlayer_15, properties {soundWaveAssetPtr=&/husk_body_impact_mono_10}}
 {id=WavePlayer_16, properties {soundWaveAssetPtr=&/husk_body_impact_mono_09}}
 {id=WavePlayer_17, properties {soundWaveAssetPtr=&/husk_body_impact_mono_08}}
 {id=WavePlayer_18, properties {soundWaveAssetPtr=&/husk_body_impact_mono_05}}
 {id=WavePlayer_19, properties {soundWaveAssetPtr=&/husk_body_impact_mono_01}}
 {id=WavePlayer_20, properties {soundWaveAssetPtr=&/husk_body_impact_mono_13}}
 {id=WavePlayer_21, properties {soundWaveAssetPtr=&/husk_body_impact_mono_12}}
 {id=WavePlayer_22, properties {soundWaveAssetPtr=&bullet_whip_in_01}}
 {id=WavePlayer_23, properties {soundWaveAssetPtr=&bullet_whip_in_02}}
 {id=WavePlayer_24, properties {soundWaveAssetPtr=&bullet_whip_in_03}}
 {id=WavePlayer_26, properties {soundWaveAssetPtr=&husk_body_impact_mono_10}}
 {id=WavePlayer_27, properties {soundWaveAssetPtr=&husk_body_impact_mono_09}}
 {id=WavePlayer_30, properties {soundWaveAssetPtr=&husk_body_impact_mono_01}}
 {id=WavePlayer_31, properties {soundWaveAssetPtr=&husk_body_impact_mono_13}}
]
properties {
 attenuationSettings=/Game/Sounds/Attenuation/Fort_Bullet_ImpactOpponent_Lng_Att
 maxDistance=5500
 soundClassObject=/Game/Sounds/FortSoundClassesAndModes/Fort_Enemy_Damaged_PSM
 volumeMultiplier=1
 attenuationOverrides {falloffDistance=5000, attenuationShapeExtents {x=1000}}
 concurrencyOverrides {maxCount=4, volumeScale=0.8}
 concurrencySet [&RifleBulletImpact_Conc]
}
