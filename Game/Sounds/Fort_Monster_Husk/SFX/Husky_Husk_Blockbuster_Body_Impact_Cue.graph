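# Sound cue for Blockbuster Husk body impacts, routed to the enemy-damaged sound class with Husk attenuation.
# A quality-level switch steps from full to simplified branches.
# The two richest layer randomized body impacts with metal pole and quieter, down-pitched cannon impacts, while the leaner two keep only pole hits.
# Modulators vary pitch per layer, and concurrency references a one-per-owner, stop-oldest rule.
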SoundCue /Game/Sounds/Fort_Monster_Husk/SFX/Husky_Husk_Blockbuster_Body_Impact_Cue
root=QualityLevel_1
nodes [
 {id=Mixer_0, children [Random_0,Modulator_3], properties {inputVolume [1,0.7]}}
 {id=Mixer_1, children [Modulator_1,Modulator_10], properties {inputVolume [1,1]}}
 {id=Mixer_2, children [Random_4,Modulator_12], properties {inputVolume [1,0.7]}}
 {id=Mixer_3, children [Modulator_0,Modulator_2], properties {inputVolume [1,1]}}
 {id=Modulator_0, children [Mixer_0], properties {pitchMax=1.07, pitchMin=0.92}}
 {id=Modulator_1, children [Random_1], properties {pitchMax=1.2}}
 {id=Modulator_10, children [Mixer_2], properties {pitchMax=1.07, pitchMin=0.92}}
 {id=Modulator_12, children [Random_24], properties {pitchMax=0.95, pitchMin=0.85}}
 {id=Modulator_2, children [Random_2], properties {pitchMax=1.2}}
 {id=Modulator_3, children [WavePlayer_3], properties {pitchMax=0.95, pitchMin=0.85}}
 {id=Modulator_4, children [Random_3], properties {pitchMax=1.07, pitchMin=0.92}}
 {id=Modulator_5, children [WavePlayer_11], properties {pitchMax=1.1, pitchMin=0.9}}
 {id=QualityLevel_1, children [Mixer_1,Mixer_3,Modulator_4,Modulator_5]}
 {id=Random_0, children [WavePlayer_0,WavePlayer_1], properties {weights [1,1]}}
 {id=Random_1, children [WavePlayer_5,WavePlayer_6,WavePlayer_7,WavePlayer_8,WavePlayer_9], properties {weights [1,1,1,1,1]}}
 {id=Random_2, children [WavePlayer_10, WavePlayer_12, WavePlayer_13], properties {weights [1,1,1]}}
 {id=Random_24, children [WavePlayer_87, WavePlayer_84], properties {weights [1,1]}}
 {id=Random_3, children [WavePlayer_4,WavePlayer_2], properties {weights [1,1]}}
 {id=Random_4, children [WavePlayer_36, WavePlayer_37], properties {weights [1,1]}}
 {id=WavePlayer_0, properties {soundWaveAssetPtr=/Game/Sounds/Fort_Impact_Sounds/Demolition/Metal/Metal_Pole_Impact01}}
 {id=WavePlayer_1, properties {soundWaveAssetPtr=/Game/Sounds/Fort_Impact_Sounds/Demolition/Metal/Metal_Pole_Impact06}}
 {id=WavePlayer_10, properties {soundWaveAssetPtr=&/husky_husk_body_impact_mix_01}}
 {id=WavePlayer_11, properties {soundWaveAssetPtr=&Metal_Pole_Impact01}}
 {id=WavePlayer_12, properties {soundWaveAssetPtr=&/husky_husk_body_impact_mix_03}}
 {id=WavePlayer_13, properties {soundWaveAssetPtr=&/husky_husk_body_impact_mix_04}}
 {id=WavePlayer_2, properties {soundWaveAssetPtr=&Metal_Pole_Impact06}}
 {id=WavePlayer_3, properties {soundWaveAssetPtr=/Game/Sounds/Fort_Impact_Sounds/Demolition/Metal/Metal_Cannon_Impact03}}
 {id=WavePlayer_36, properties {soundWaveAssetPtr=&Metal_Pole_Impact01}}
 {id=WavePlayer_37, properties {soundWaveAssetPtr=&Metal_Pole_Impact06}}
 {id=WavePlayer_4, properties {soundWaveAssetPtr=&Metal_Pole_Impact01}}
 {id=WavePlayer_5, properties {soundWaveAssetPtr=&husky_husk_body_impact_mix_01}}
 {id=WavePlayer_6, properties {soundWaveAssetPtr=&/husky_husk_body_impact_mix_02}}
 {id=WavePlayer_7, properties {soundWaveAssetPtr=&husky_husk_body_impact_mix_03}}
 {id=WavePlayer_8, properties {soundWaveAssetPtr=&husky_husk_body_impact_mix_04}}
 {id=WavePlayer_84, properties {soundWaveAssetPtr=/Game/Sounds/Fort_Impact_Sounds/Demolition/Metal/Metal_Cannon_Impact02}}
 {id=WavePlayer_87, properties {soundWaveAssetPtr=&Metal_Cannon_Impact03}}
 {id=WavePlayer_9, properties {soundWaveAssetPtr=&/husky_husk_body_impact_mix_05}}
]
properties {
 attenuationSettings=/Game/Sounds/Attenuation/Fort_Monster_Husk_Att
 maxDistance=2800
 soundClassObject=/Game/Sounds/FortSoundClassesAndModes/Fort_Enemy_Damaged_PSM
 volumeMultiplier=0.5
 attenuationOverrides {falloffDistance=5000, attenuationShapeExtents {x=1000}}
 concurrencyOverrides {resolutionRule=PreventNew}
 concurrencySet [/Game/Sounds/Concurrency_Rules/OneMaximum_Owner_StopOldest]
}
