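# Sound cue for Husky Husk body impacts, using the enemy-damaged sound class and Husk attenuation.
# A quality-level root selects five equally weighted wave variations at the first level or a three-wave subset at the second, leaving the other two slots empty.
# Each populated branch runs its random pick through a pitch modulator for variation.
# A concurrency rule named for one instance per owner, stopping the oldest, limits overlap.
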SoundCue /Game/Sounds/Fort_Monster_Husk/SFX/Husky_Husk_Body_Impact_Cue
root=QualityLevel_1
nodes [
 {id=Modulator_1, children [Random_1], properties {pitchMax=1.2}}
 {id=Modulator_2, children [Random_2], properties {pitchMax=1.2}}
 {id=QualityLevel_1, children [Modulator_1,Modulator_2,null,null]}
 {id=Random_1, children [WavePlayer_5,WavePlayer_6,WavePlayer_7,WavePlayer_8,WavePlayer_9], properties {weights [1,1,1,1,1]}}
 {id=Random_2, children [WavePlayer_10, WavePlayer_12, WavePlayer_13], properties {weights [1,1,1]}}
 {id=WavePlayer_10, properties {soundWaveAssetPtr=&/husky_husk_body_impact_mix_01}}
 {id=WavePlayer_12, properties {soundWaveAssetPtr=&/husky_husk_body_impact_mix_03}}
 {id=WavePlayer_13, properties {soundWaveAssetPtr=&/husky_husk_body_impact_mix_04}}
 {id=WavePlayer_5, properties {soundWaveAssetPtr=&husky_husk_body_impact_mix_01}}
 {id=WavePlayer_6, properties {soundWaveAssetPtr=&/husky_husk_body_impact_mix_02}}
 {id=WavePlayer_7, properties {soundWaveAssetPtr=&husky_husk_body_impact_mix_03}}
 {id=WavePlayer_8, properties {soundWaveAssetPtr=&husky_husk_body_impact_mix_04}}
 {id=WavePlayer_9, properties {soundWaveAssetPtr=&/husky_husk_body_impact_mix_05}}
]
properties {
 attenuationSettings=/Game/Sounds/Attenuation/Fort_Monster_Husk_Att
 maxDistance=2800
 soundClassObject=/Game/Sounds/FortSoundClassesAndModes/Fort_Enemy_Damaged_PSM
 volumeMultiplier=2
 attenuationOverrides {falloffDistance=5000, attenuationShapeExtents {x=1000}}
 concurrencyOverrides {resolutionRule=PreventNew}
 concurrencySet [/Game/Sounds/Concurrency_Rules/OneMaximum_Owner_StopOldest]
}
