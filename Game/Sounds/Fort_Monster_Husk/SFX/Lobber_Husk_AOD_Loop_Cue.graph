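# Looping sound cue for the Lobber Husk's area-of-denial poison gas, routed to the ranged weapon projectile effects sound class.
# A quality-level root selects between two mixers, with its last two slots empty.
# Each mixer layers a randomly picked gas loop with a half-volume stereo gas bed; the second picks from two variants instead of three.
# It uses shared close-enemy Beehive attenuation, and a concurrency override blocks new instances beyond the cap.
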SoundCue /Game/Sounds/Fort_Monster_Husk/SFX/Lobber_Husk_AOD_Loop_Cue
root=QualityLevel_0
nodes [
 {id=Mixer_1, children [Random_1,WavePlayer_4], properties {inputVolume [1,0.5]}}
 {id=Mixer_2, children [Random_2,WavePlayer_8], properties {inputVolume [1,0.5]}}
 {id=QualityLevel_0, children [Mixer_1,Mixer_2,null,null]}
 {id=Random_1, children [WavePlayer_5,WavePlayer_6,WavePlayer_7], properties {weights [1,1,1]}}
 {id=Random_2, children [WavePlayer_9, WavePlayer_11], properties {weights [1,1]}}
 {id=WavePlayer_11, properties {bLooping=true, soundWaveAssetPtr=&/lobber_poison_gasloop_03Loop}}
 {id=WavePlayer_4, properties {bLooping=true, soundWaveAssetPtr=&/area_of_denial_gas_stereo_01Loop}}
 {id=WavePlayer_5, properties {bLooping=true, soundWaveAssetPtr=&/lobber_poison_gasloop_01Loop}}
 {id=WavePlayer_6, properties {bLooping=true, soundWaveAssetPtr=&/lobber_poison_gasloop_02Loop}}
 {id=WavePlayer_7, properties {bLooping=true, soundWaveAssetPtr=&lobber_poison_gasloop_03Loop}}
 {id=WavePlayer_8, properties {bLooping=true, soundWaveAssetPtr=&area_of_denial_gas_stereo_01Loop}}
 {id=WavePlayer_9, properties {bLooping=true, soundWaveAssetPtr=&lobber_poison_gasloop_01Loop}}
]
properties {
 attenuationSettings=/Game/Sounds/Attenuation/Fort_Beehive_Enemy_Close_Att
 bOverrideConcurrency=true
 maxDistance=1000
 soundClassObject=&Fort_WeaponRanged_Projectile_Effects
 volumeMultiplier=0.5
 attenuationOverrides {falloffDistance=3400, attenuationShapeExtents {x=600}}
 concurrencyOverrides {maxCount=6, resolutionRule=PreventNew}
}
