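# Sound cue for the impact of a ranged Husk's bone projectile, routed to the ranged projectile effects sound class.
# Each mix layers pitch-modulated random bone impacts and Husk bone sounds with an unmodulated random Technician impact sound.
# A quality-level root selects between two such mixes, the second with fewer variations, and leaves its other two slots empty.
# Small melee weapon attenuation and a concurrency override limit its range and simultaneous instances.
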SoundCue /Game/Sounds/Fort_Monster_Husk/SFX/Ranged_Husk_Bone_Impact_Cue
root=QualityLevel_0
nodes [
 {id=Mixer_1, children [Modulator_2,Modulator_3,Random_5], properties {inputVolume [1,1,1]}}
 {id=Mixer_2, children [Modulator_4,Modulator_5,Random_8], properties {inputVolume [1,1,1]}}
 {id=Modulator_2, children [Random_3], properties {pitchMax=1.3}}
 {id=Modulator_3, children [Random_4], properties {pitchMax=1.3}}
 {id=Modulator_4, children [Random_6], properties {pitchMax=1.3}}
 {id=Modulator_5, children [Random_7], properties {pitchMax=1.3}}
 {id=QualityLevel_0, children [Mixer_1,Mixer_2,null,null]}
 {id=Random_3, children [WavePlayer_9, WavePlayer_10, WavePlayer_11, WavePlayer_12], properties {weights [1,1,1,1]}}
 {id=Random_4, children [WavePlayer_13, WavePlayer_14, WavePlayer_15], properties {weights [1,1,1]}}
 {id=Random_5, children [WavePlayer_17, WavePlayer_16], properties {weights [1,1]}}
 {id=Random_6, children [WavePlayer_18, WavePlayer_19, WavePlayer_21], properties {weights [1,1,1]}}
 {id=Random_7, children [WavePlayer_22, WavePlayer_24], properties {weights [1,1]}}
 {id=Random_8, children [WavePlayer_26, WavePlayer_25], properties {weights [1,1]}}
 {id=WavePlayer_10, properties {soundWaveAssetPtr=&/ranged_husk_bone_impact_02}}
 {id=WavePlayer_11, properties {soundWaveAssetPtr=&/ranged_husk_bone_impact_03}}
 {id=WavePlayer_12, properties {soundWaveAssetPtr=&/ranged_husk_bone_impact_04}}
 {id=WavePlayer_13, properties {soundWaveAssetPtr=&/Husk_Bones01}}
 {id=WavePlayer_14, properties {soundWaveAssetPtr=&/Husk_Bones05}}
 {id=WavePlayer_15, properties {soundWaveAssetPtr=&/Husk_Bones06}}
 {id=WavePlayer_16, properties {soundWaveAssetPtr=/Game/Sounds/Fort_Gadgets/Technician/Impact_Sweet_K_07b}}
 {id=WavePlayer_17, properties {soundWaveAssetPtr=/Game/Sounds/Fort_Gadgets/Technician/Impact_Sweet_K_02a}}
 {id=WavePlayer_18, properties {soundWaveAssetPtr=&/ranged_husk_bone_impact_01}}
 {id=WavePlayer_19, properties {soundWaveAssetPtr=&ranged_husk_bone_impact_02}}
 {id=WavePlayer_21, properties {soundWaveAssetPtr=&ranged_husk_bone_impact_04}}
 {id=WavePlayer_22, properties {soundWaveAssetPtr=&Husk_Bones01}}
 {id=WavePlayer_24, properties {soundWaveAssetPtr=&Husk_Bones06}}
 {id=WavePlayer_25, properties {soundWaveAssetPtr=&Impact_Sweet_K_07b}}
 {id=WavePlayer_26, properties {soundWaveAssetPtr=&Impact_Sweet_K_02a}}
 {id=WavePlayer_9, properties {soundWaveAssetPtr=&ranged_husk_bone_impact_01}}
]
properties {
 attenuationSettings=&Fort_Weapon_Melee_Sml_Att
 bOverrideConcurrency=true
 maxDistance=2800
 soundClassObject=&Fort_WeaponRanged_Projectile_Effects
 volumeMultiplier=1
 concurrencyOverrides {maxCount=3}
}
