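# Sound cue that plays a random yell when a Husk monster spawns.
# A quality-level switch feeds two equally weighted random pools, six variations in the first slot and a reduced set of three in the second, with the remaining slots empty.
# It uses the enemy spawn sound class and Husk spawn attenuation, and its concurrency override allows one instance, blocking new plays while one is active.
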SoundCue /Game/Sounds/Fort_Monster_Husk/SFX/husk_spawn_yells_Cue
root=QualityLevel_0
nodes [
 {id=QualityLevel_0, children [Random_1,Random_2,null,null]}
 {
  id=Random_1
  children [WavePlayer_6, WavePlayer_7, WavePlayer_8, WavePlayer_9, WavePlayer_10, WavePlayer_11]
  properties {weights [1,1,1,1,1,1]}
 }
 {id=Random_2, children [WavePlayer_12, WavePlayer_13, WavePlayer_14], properties {weights [1,1,1]}}
 {id=WavePlayer_10, properties {soundWaveAssetPtr=&/husk_spawn_yells_05}}
 {id=WavePlayer_11, properties {soundWaveAssetPtr=&/husk_spawn_yells_06}}
 {id=WavePlayer_12, properties {soundWaveAssetPtr=&/husk_spawn_yells_01}}
 {id=WavePlayer_13, properties {soundWaveAssetPtr=&/husk_spawn_yells_02}}
 {id=WavePlayer_14, properties {soundWaveAssetPtr=&/husk_spawn_yells_03}}
 {id=WavePlayer_6, properties {soundWaveAssetPtr=&husk_spawn_yells_01}}
 {id=WavePlayer_7, properties {soundWaveAssetPtr=&husk_spawn_yells_02}}
 {id=WavePlayer_8, properties {soundWaveAssetPtr=&husk_spawn_yells_03}}
 {id=WavePlayer_9, properties {soundWaveAssetPtr=&/husk_spawn_yells_04}}
]
properties {
 attenuationSettings=/Game/Sounds/Attenuation/Fort_Monster_Husk_Spawn_Att
 bOverrideConcurrency=true
 maxDistance=9500
 soundClassObject=/Game/Sounds/FortSoundClassesAndModes/Fort_Enemy_Spawn
 volumeMultiplier=1
 concurrencyOverrides {maxCount=1, resolutionRule=PreventNew}
}
