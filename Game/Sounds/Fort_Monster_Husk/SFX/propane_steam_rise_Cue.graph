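# Sound cue for a rising propane steam effect in the Husk monster SFX set, routed through the enemy attack sound class.
# A single wave player feeds a Modulator that randomizes playback, with a raised upper pitch limit.
# It uses the Husk attenuation settings, a boosted cue volume and a concurrency override that allows only two simultaneous instances.
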SoundCue /Game/Sounds/Fort_Monster_Husk/SFX/propane_steam_rise_Cue
root=Modulator_0
nodes [
 {id=Modulator_0, children [WavePlayer_0], properties {pitchMax=1.3}}
 {id=WavePlayer_0, properties {soundWaveAssetPtr=&/propane_steam_rise}}
]
properties {
 attenuationSettings=/Game/Sounds/Attenuation/Fort_Monster_Husk_Att
 bOverrideConcurrency=true
 maxDistance=2800
 soundClassObject=/Game/Sounds/FortSoundClassesAndModes/Fort_Enemy_Attack
 volumeMultiplier=1.5
 concurrencyOverrides {maxCount=2}
}
